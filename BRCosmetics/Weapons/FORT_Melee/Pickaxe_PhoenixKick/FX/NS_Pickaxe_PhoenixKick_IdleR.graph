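# Idle Niagara effect for the right-hand PhoenixKick pickaxe cosmetic, layering meshes, a light, sparks and water sprites.
# The headwater_R emitter bursts two mirrored mesh particles that track a skeletal mesh, while Sparks and Water continuously spawn drifting sprites in two offset streams.
# User.VariantSwitch drives three color variants through array lookups and color-curve lerps.
# Front-end state adjusts spark size, brightness and mesh glow, and time of day also scales mesh color alpha.
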NiagaraSystem /BRCosmetics/Weapons/FORT_Melee/Pickaxe_PhoenixKick/FX/NS_Pickaxe_PhoenixKick_IdleR
dependencies {
 usedDataInterfaces [
  &NiagaraDataInterfaceArrayColor
  &NiagaraDataInterfaceArrayFloat3
  &NiagaraDataInterfaceArrayFloat
  &NiagaraDataInterfaceCurve
  &NiagaraDataInterfaceColorCurve
  &NiagaraDataInterfaceSpriteRendererInfo
  &NiagaraDataInterfaceSkeletalMesh
 ]
 usedDynamicInputs [&Vector2DFromFloat]
 usedModules [
  &self:SystemSpawnScript.SystemScriptSource.SystemScriptGraph.NiagaraNodeAssignment_0.SetVariables_#1
  &SystemState
 ]
}
emitters {
 Light {
  emitterUpdateScript [
   &EmitterState(
    Inactive Response=::NewEnumerator1
    Life Cycle Mode=::NewEnumerator1
    Loop Behavior=::NewEnumerator1
    Loop Duration Mode=::NewEnumerator1
    UseLoopDelay=false
   )
   &SpawnBurst_Instantaneous(Use Spawn Probability=false, UseLoopCountLimit=false)
  ]
  particleSpawnScript [&InitializeParticle(Color=<rgba 0.7 0.7 0.7 1>, Position Offset=<0 -15 0>, UsePositionOffset=true, Write Material Random=true)]
  particleUpdateScript [
   &ParticleState(Kill Particles When Lifetime Has Elapsed=false)
   &FN_Light_Attributes(
    Color=&SelectLinearColorFromArray(
     Array Sampling Mode=::NewEnumerator1
     Direct Array Index=&MakeIntFromFloat(Float=$User.VariantSwitch)
     Color Selection Array=DataInterface() {propertyValues {ColorData [<rgba 0.3986 5 3.5409 1>, <rgba 0.7151 3.6445 10 1>, <rgba 5 3.4409 0.6583 1>]}}
     Random Seed=?
    )
    Create Color Attribute=true
    Create Exponent Attribute=true
    Create Position Attribute=true
    Create Radius Attribute=true
    Enable Particle Light Enable=true
    Exponent=2
    Position:NiagaraPosition=$Particles.SkeletalMeshLocation.SampledPosition
    Radius=60
    Scale Color=4
   )
  ]
  rendererClasses [&NiagaraLightRendererProperties, &NiagaraSpriteRendererProperties]
  renderers [
   {rendererClass=&NiagaraLightRendererProperties, rendererIndex=0}
   {rendererClass=&NiagaraSpriteRendererProperties, rendererIndex=1}
  ]
 }
 Sparks {
  emitterUpdateScript [
   &EmitterState(Life Cycle Mode=::NewEnumerator1, Loop Duration=1, UseLoopDelay=false)
   &SpawnRate(SpawnRate=20, Use Spawn Probability=false)
   /SetVariables(Emitter.Color_BaseColor:LinearColor=&SelectLinearColorFromArray(
     Array Sampling Mode=::NewEnumerator1
     Direct Array Index=&MakeIntFromFloat(Float=$User.VariantSwitch)
     Color Selection Array=DataInterface() {propertyValues {ColorData [<rgba 0.0092 0.1 0.0748 1>, <rgba 0.0092 0.1 0.0748 1>, <rgba 0.1 0.0667 0.0102 1>]}}
     Random Seed=?
    )) {isSetParameters=true}
   /SetVariables(Emitter.Color_DarkColor:LinearColor=&SelectLinearColorFromArray(
     Array Sampling Mode=::NewEnumerator1
     Direct Array Index=&MakeIntFromFloat(Float=$User.VariantSwitch)
     Color Selection Array=DataInterface() {propertyValues {ColorData [<rgba 0.0024 0.02 0.0194 1>, <rgba 0.0024 0.02 0.0194 1>, <rgba 0.02 0.0094 0.0019 1>]}}
     Random Seed=?
    )) {isSetParameters=true}
   /SetVariables(Emitter.Color_LightColor:LinearColor=&SelectLinearColorFromArray(
     Array Sampling Mode=::NewEnumerator1
     Direct Array Index=&MakeIntFromFloat(Float=$User.VariantSwitch)
     Color Selection Array=DataInterface() {propertyValues {ColorData [<rgba 0.3177 1 0.8204 1>, <rgba 0.9531 3 2.4612 1>, <rgba 5 3.9197 1.3541 1>]}}
     Random Seed=?
    )) {isSetParameters=true}
  ]
  particleSpawnScript [
   &PartitionParticles()
   &InitializeParticle(
    Color=&MakeCustomLinearColorFromBool(Bool=&RandomBool(), True Color=<rgba 1 1 1 1.5>, False Color=<rgba 1.5 1.5 1.5 1>)
    Lifetime=0.8
    Sprite Size Max=<2.5 5>
    Sprite Size Min=<1.5 3>
    Sprite Size Mode=::NewEnumerator2
    UsePositionOffset=false
    Write Material Random=true
   )
   &ShapeLocation(
    MaskSpawnGroup=false
    Offset=&SelectVectorFromArray(
     Direct Array Index=$Particles.PartitionParticles.Partition
     Vector Selection Array=DataInterface() {propertyValues {FloatData [<-6 -6 10>, <-6 -6 -10>]}}
    )
    Offset Mode=::NewEnumerator0
    Override Seed=false
    Sphere Radius=4
    UseOwnerScale=true
   )
   &AddVelocity(Velocity=&SelectVectorFromArray(
     Direct Array Index=$Particles.PartitionParticles.Partition
     Vector Selection Array=DataInterface() {propertyValues {FloatData [<40 50 5>, <40 50 -5>]}}
    ))
   &AddVelocity(Module.Use Falloff Distance=false, Override Seed=false, Velocity Mode=::NewEnumerator1, Velocity Speed=15) {module=AddVelocity001}
  ]
  particleUpdateScript [
   &ParticleState()
   &CurlNoiseForce(Bias Noise=false, Noise Strength=50, Pan Noise=false)
   &Drag(Drag=0.25, Use Linear Drag=true, Use Rotational Drag=true)
   &SolveForcesAndVelocity(Clamp Velocity=false, Limit Acceleration=false)
   &ScaleSpriteSize(
    Uniform Curve Scale=&MakeCustomFloatFromBool(Bool=$User.bIsFrontEndPreview, True Float=0.7, False Float=1)
    Uniform Curve Sprite Scale=DataInterface() {propertyValues {Curve {keys [
        <at 0 0 RCIM_Cubic>
        {arriveTangent=1.5385, interpMode=RCIM_Cubic, leaveTangent=0.0444, tangentMode=RCTM_Break, time=0.1, value=1}
        {arriveTangent=0.0326, interpMode=RCIM_Cubic, leaveTangent=-6.725, tangentMode=RCTM_Break, time=0.65, value=1}
        <at 1 0 RCIM_Cubic>
       ]}}}
   )
   &ScaleColor(
    Scale RGB=&Multiply_VectorByFloat(
     Vector=/Niagara/DynamicInputs/Math/LerpMultipleVectors(
      Alpha:NiagaraFloat=&Divide_Float(A=$User.VariantSwitch, B=2)
      V 1:Vector3f=&MakeVectorFromLinearColorRGB(Color=&ColorFromCurve(ColorCurve=DataInterface() {propertyValues {
          AlphaCurve {keys []}
          BlueCurve {keys [<at 0.4436 0.5573 RCIM_Cubic>,<at 0.6697 1 RCIM_Cubic>]}
          GreenCurve {keys [<at 0.4436 1 RCIM_Cubic>,<at 0.6697 0.6733 RCIM_Cubic>]}
          RedCurve {keys [<at 0.4436 0.5895 RCIM_Cubic>,<at 0.6697 0.2621 RCIM_Cubic>]}
         }}))
      V 2:Vector3f=&MakeVectorFromLinearColorRGB(Color=&ColorFromCurve(ColorCurve=DataInterface() {propertyValues {
          AlphaCurve {keys []}
          BlueCurve {keys [<at 0.4436 2 RCIM_Cubic>,<at 0.6697 1 RCIM_Cubic>]}
          GreenCurve {keys [<at 0.4436 0.5697 RCIM_Cubic>,<at 0.6697 0.4979 RCIM_Cubic>]}
          RedCurve {keys [<at 0.4436 0.1469 RCIM_Cubic>,<at 0.6697 0.186 RCIM_Cubic>]}
         }}))
      V 3:Vector3f=&MakeVectorFromLinearColorRGB(Color=&ColorFromCurve(ColorCurve=DataInterface() {propertyValues {AlphaCurve {keys []}, BlueCurve {keys [<at 0.132 1.5>]}, GreenCurve {keys [<at 0.132 1.5>]}, RedCurve {keys [<at 0.132 1.5>]}}}))
      V 4:Vector3f=&MakeVectorFromLinearColorRGB(Color=&ColorFromCurve(ColorCurve=DataInterface() {propertyValues {
          AlphaCurve {keys []}
          BlueCurve {keys [<at 0 0 RCIM_Cubic>,<at 0.132 0>,<at 0.52 0 RCIM_Cubic>]}
          GreenCurve {keys [<at 0 0.2552 RCIM_Cubic>,<at 0.132 1>,<at 0.52 0.1406 RCIM_Cubic>]}
          RedCurve {keys [<at 0 0.145 RCIM_Cubic>,<at 0.132 0.8196>,<at 0.52 0.062 RCIM_Cubic>]}
         }}))
      V 5:Vector3f=&MakeVectorFromLinearColorRGB(Color=&ColorFromCurve(ColorCurve=DataInterface() {propertyValues {
          AlphaCurve {keys []}
          BlueCurve {keys [<at 0 0.2552 RCIM_Cubic>,<at 0.132 1>,<at 0.52 0.1042 RCIM_Cubic>]}
          GreenCurve {keys [<at 0 0.0325 RCIM_Cubic>,<at 0.132 0.1094>,<at 0.52 0.0553 RCIM_Cubic>]}
          RedCurve {keys [<at 0 0.0577 RCIM_Cubic>,<at 0.132 0.5547>,<at 0.52 0.0607 RCIM_Cubic>]}
         }}))
      Vector Count:ENiagara_MultipleLerpCount=::NewEnumerator0
     )
     Float=&FloatFromCurve(
      FloatCurve=DataInterface() {propertyValues {Curve {keys [
          {interpMode=RCIM_Cubic, tangentMode=RCTM_User, time=0, value=1}
          {arriveTangent=-0.3422, interpMode=RCIM_Cubic, leaveTangent=-0.3422, tangentMode=RCTM_User, time=0.3, value=0.925}
          {interpMode=RCIM_Cubic, tangentMode=RCTM_User, time=1, value=0.85}
         ]}}}
      Scale Curve=&MakeCustomFloatFromBool(Bool=$User.bIsFrontEndPreview, True Float=12, False Float=8)
     )
    )
    ScaleA=true
    ScaleRGB=true
   )
  ]
  rendererClasses [&NiagaraSpriteRendererProperties]
  renderers [{rendererClass=&NiagaraSpriteRendererProperties}]
 }
 Water {
  emitterUpdateScript [
   &EmitterState(Life Cycle Mode=::NewEnumerator1, Loop Duration=1, UseLoopDelay=false)
   &SpawnRate(SpawnRate=30, Use Spawn Probability=false)
  ]
  particleSpawnScript [
   &PartitionParticles()
   &InitializeParticle(
    Color=&MakeCustomLinearColorFromBool(Bool=&RandomBool(), True Color=<rgba 1 1 1 1.5>, False Color=<rgba 1.5 1.5 1.5 1>)
    Sprite Rotation Mode=::NewEnumerator2
    Sprite Size Mode=::NewEnumerator1
    Sprite UV Mode=::NewEnumerator3
    Uniform Sprite Size Max=20
    Uniform Sprite Size Min=10
    UsePositionOffset=false
    Write Material Random=true
   )
   &SubUVAnimation(
    End Frame Range Override=3
    "Sprite Renderer "=DataInterface() {propertyValues {SpriteRenderer=&self:Water_0.NiagaraSpriteRendererProperties_0}}
    SubUV Animation Mode:ENiagara_SubUVLookupModeV2=::NewEnumerator2
    UseEndFrame=true
    UseStartFrame=false
   )
   &ShapeLocation(
    MaskSpawnGroup=false
    Offset=&SelectVectorFromArray(
     Direct Array Index=$Particles.PartitionParticles.Partition
     Vector Selection Array=DataInterface() {propertyValues {FloatData [<-6 -8 8>, <-6 -8 -8>]}}
    )
    Offset Mode=::NewEnumerator0
    Override Seed=false
    Sphere Radius=4
    UseOwnerScale=true
   )
   &AddVelocity(Velocity=&SelectVectorFromArray(
     Direct Array Index=$Particles.PartitionParticles.Partition
     Vector Selection Array=DataInterface() {propertyValues {FloatData [<25 40 5>, <25 40 -5>]}}
    ))
   &AddVelocity(Module.Use Falloff Distance=false, Override Seed=false, Velocity Mode=::NewEnumerator1, Velocity Speed=15) {module=AddVelocity001}
  ]
  particleUpdateScript [
   &ParticleState()
   &ScaleSpriteSize(Uniform Curve Sprite Scale=DataInterface() {propertyValues {Curve {keys [
        <at 0 0 RCIM_Cubic>
        {arriveTangent=0.0188, interpMode=RCIM_Cubic, leaveTangent=0.75, tangentMode=RCTM_Break, time=0.25, value=0}
        {arriveTangent=1.3333, interpMode=RCIM_Cubic, leaveTangent=1.3333, time=0.3, value=0.5}
        <at 1 1 RCIM_Cubic>
       ]}}})
   &DynamicMaterialParameters(
    Index 0 Param 1=&FloatFromCurve(FloatCurve=DataInterface() {propertyValues {Curve {
        defaultValue=0
        keys [
         {arriveTangent=0.0176, interpMode=RCIM_Cubic, leaveTangent=0.0176, tangentMode=RCTM_User, time=0, value=1}
         {arriveTangent=-1, interpMode=RCIM_Cubic, leaveTangent=-1, time=0.5, value=0.8}
         {arriveTangent=-0.9176, interpMode=RCIM_Cubic, leaveTangent=-0.9176, tangentMode=RCTM_User, time=1, value=0}
        ]
       }}})
    Module.Index0Float0Write=true
    Module.Index0Float1Write=false
    Module.Index0Float2Write=false
    Module.Index0Float3Write=false
   )
   &ScaleColor(
    Scale Alpha=&FloatFromCurve(FloatCurve=DataInterface() {propertyValues {Curve {keys [<at 0 0>,<at 0.1 1>,<at 0.8 1>,<at 1 0>]}}}, Scale Curve=1.2)
    Scale RGB=&LerpMultipleVectors(
     Alpha:NiagaraFloat=&Divide_Float(A=$User.VariantSwitch, B=2)
     V 1:Vector3f=&MakeVectorFromLinearColorRGB(Color=&ColorFromCurve(ColorCurve=DataInterface() {propertyValues {
         AlphaCurve {keys []}
         BlueCurve {keys [<at 0.4465 0.6159 RCIM_Cubic>,<at 0.6131 1.2 RCIM_Cubic>]}
         GreenCurve {keys [<at 0.4465 1 RCIM_Cubic>,<at 0.6131 0.6544 RCIM_Cubic>]}
         RedCurve {keys [<at 0.4465 0.2969 RCIM_Cubic>,<at 0.6131 0.1875 RCIM_Cubic>]}
        }}))
     V 2:Vector3f=&MakeVectorFromLinearColorRGB(Color=&ColorFromCurve(ColorCurve=DataInterface() {propertyValues {
         AlphaCurve {keys []}
         BlueCurve {keys [<at 0.3467 1 RCIM_Cubic>,<at 0.5482 1.2 RCIM_Cubic>]}
         GreenCurve {keys [<at 0.3467 0.5989 RCIM_Cubic>,<at 0.5482 0.454 RCIM_Cubic>]}
         RedCurve {keys [<at 0.3467 0.2448 RCIM_Cubic>,<at 0.5482 0.173 RCIM_Cubic>]}
        }}))
     V 3:Vector3f=&MakeVectorFromLinearColorRGB(Color=&ColorFromCurve(ColorCurve=DataInterface() {propertyValues {
         AlphaCurve {keys []}
         BlueCurve {keys [<at 0.347 0.4583>,<at 0.5663 0.1428>]}
         GreenCurve {keys [<at 0.347 0.9051>,<at 0.5663 0.6057>]}
         RedCurve {keys [<at 0.347 1>,<at 0.5663 1.2>]}
        }}))
     V 4:Vector3f=&MakeVectorFromLinearColorRGB(Color=&ColorFromCurve(ColorCurve=DataInterface() {propertyValues {
         AlphaCurve {keys []}
         BlueCurve {keys [<at 0 0 RCIM_Cubic>,<at 0.132 0>,<at 0.52 0 RCIM_Cubic>]}
         GreenCurve {keys [<at 0 0.2552 RCIM_Cubic>,<at 0.132 1>,<at 0.52 0.1406 RCIM_Cubic>]}
         RedCurve {keys [<at 0 0.145 RCIM_Cubic>,<at 0.132 0.8196>,<at 0.52 0.062 RCIM_Cubic>]}
        }}))
     V 5:Vector3f=&MakeVectorFromLinearColorRGB(Color=&ColorFromCurve(ColorCurve=DataInterface() {propertyValues {
         AlphaCurve {keys []}
         BlueCurve {keys [<at 0 0.2552 RCIM_Cubic>,<at 0.132 1>,<at 0.52 0.1042 RCIM_Cubic>]}
         GreenCurve {keys [<at 0 0.0325 RCIM_Cubic>,<at 0.132 0.1094>,<at 0.52 0.0553 RCIM_Cubic>]}
         RedCurve {keys [<at 0 0.0577 RCIM_Cubic>,<at 0.132 0.5547>,<at 0.52 0.0607 RCIM_Cubic>]}
        }}))
     Vector Count:ENiagara_MultipleLerpCount=::NewEnumerator0
    )
    ScaleA=true
    ScaleRGB=true
   )
   &SolveForcesAndVelocity(Clamp Velocity=false, Limit Acceleration=false)
  ]
  rendererClasses [&NiagaraSpriteRendererProperties]
  renderers [{rendererClass=&NiagaraSpriteRendererProperties}]
 }
 headwater_R {
  emitterUpdateScript [
   &EmitterState(
    Inactive Response=::NewEnumerator1
    Life Cycle Mode=::NewEnumerator1
    Loop Behavior=::NewEnumerator1
    Loop Duration Mode=::NewEnumerator1
    UseLoopDelay=false
   )
   &SpawnBurst_Instantaneous(Spawn Count=2, Use Spawn Probability=false, UseLoopCountLimit=false)
   /SetVariables(Emitter.Color_VariantSwitch:LinearColor=&SelectLinearColorFromArray(
     Array Sampling Mode=::NewEnumerator1
     Direct Array Index=&MakeIntFromFloat(Float=$User.VariantSwitch)
     Color Selection Array=DataInterface() {propertyValues {ColorData [<rgba 0.0454 1 0.2773 1>, <rgba 0.7 0.7 0.7 1>, <rgba 2 1.3984 0 1>]}}
     Random Seed=?
    )) {isSetParameters=true}
   /SetVariables(Emitter.Color_V_Bottom_LinearA:LinearColor=&SelectLinearColorFromArray(
     Array Sampling Mode=::NewEnumerator1
     Direct Array Index=&MakeIntFromFloat(Float=$User.VariantSwitch)
     Color Selection Array=DataInterface() {propertyValues {ColorData [<rgba 0 0.8579 2 1>, <rgba 0 0.8579 2 1>, <rgba 1.5 0.8 0.0241 1>]}}
     Random Seed=?
    )) {isSetParameters=true}
   /SetVariables(Emitter.Color_V_Dark_LinearA:LinearColor=&SelectLinearColorFromArray(
     Array Sampling Mode=::NewEnumerator1
     Direct Array Index=&MakeIntFromFloat(Float=$User.VariantSwitch)
     Color Selection Array=DataInterface() {propertyValues {ColorData [<rgba 0.0646 1 0.5517 1>, <rgba 0.0646 1 0.5517 1>, <rgba 0.8 0.8 0.8 1>]}}
     Random Seed=?
    )) {isSetParameters=true}
   /SetVariables(Emitter.Color_V_Dark_LinearB:LinearColor=&SelectLinearColorFromArray(
     Array Sampling Mode=::NewEnumerator1
     Direct Array Index=&MakeIntFromFloat(Float=$User.VariantSwitch)
     Color Selection Array=DataInterface() {propertyValues {ColorData [<rgba 0.0677 1 0.45 1>, <rgba 0.0677 1 0.45 1>, <rgba 0.8 0.8 0.8 1>]}}
     Random Seed=?
    )) {isSetParameters=true}
   /SetVariables(Emitter.Color_VariantSwitch_Light:LinearColor=&SelectLinearColorFromArray(
     Array Sampling Mode=::NewEnumerator1
     Direct Array Index=&MakeIntFromFloat(Float=$User.VariantSwitch)
     Color Selection Array=DataInterface() {propertyValues {ColorData [<rgba 0.0858 1.2 0.4373 1>, <rgba 1 1 1 1>, <rgba 1 1 1 1>]}}
     Random Seed=?
    )) {isSetParameters=true}
   /SetVariables(Emitter.Color_V_LightA:LinearColor=&SelectLinearColorFromArray(
     Array Sampling Mode=::NewEnumerator1
     Direct Array Index=&MakeIntFromFloat(Float=$User.VariantSwitch)
     Color Selection Array=DataInterface() {propertyValues {ColorData [<rgba 0.1013 1 0.2922 1>, <rgba 1 1 1 1>, <rgba 1 1 1 1>]}}
     Random Seed=?
    )) {isSetParameters=true}
   /SetVariables(Emitter.Color_V_Glow_LinearA:LinearColor=&SelectLinearColorFromArray(
     Array Sampling Mode=::NewEnumerator1
     Direct Array Index=&MakeIntFromFloat(Float=$User.VariantSwitch)
     Color Selection Array=DataInterface() {propertyValues {ColorData [<rgba 0 0.429 1 1>, <rgba 0 0.429 1 1>, <rgba 1 1 1 1>]}}
     Random Seed=?
    )) {isSetParameters=true}
   /SetVariables(Emitter.Color_V_Glow_LinearB:LinearColor=&SelectLinearColorFromArray(
     Array Sampling Mode=::NewEnumerator1
     Direct Array Index=&MakeIntFromFloat(Float=$User.VariantSwitch)
     Color Selection Array=DataInterface() {propertyValues {ColorData [<rgba 0.0677 1 0.45 1>, <rgba 0.0677 1 0.45 1>, <rgba 1 1 1 1>]}}
     Random Seed=?
    )) {isSetParameters=true}
  ]
  particleSpawnScript [
   &PartitionParticles()
   &InitializeParticle(
    Color=<rgba 0.7 0.7 0.7 1>
    Lifetime=5
    Mesh Scale=&Multiply_VectorByFloat(
     Vector=&SelectVectorFromArray(
      Direct Array Index=$Particles.PartitionParticles.Partition
      Vector Selection Array=DataInterface() {propertyValues {FloatData [<1 1 1>, <-1 1 1>]}}
     )
     Float=0.6
    )
    Mesh Scale Mode=::NewEnumerator4
    UsePositionOffset=false
    Write Material Random=true
   )
   &InitialMeshOrientation(Mesh Orientation Mode=::NewEnumerator2, Rotation=<0 0.25 0>, Use Rotation Vector=true)
   &InitialMeshOrientation(
    Mesh Orientation Mode=::NewEnumerator2
    Rotation=&SelectVectorFromArray(
     Direct Array Index=$Particles.PartitionParticles.Partition
     Vector Selection Array=DataInterface() {propertyValues {FloatData [<0.03 0 0>, <-0.05 0 0>]}}
    )
    Use Rotation Vector=true
   ) {module=InitialMeshOrientation001}
   &InitialMeshOrientation(
    Mesh Orientation Mode=::NewEnumerator2
    Rotation=&SelectVectorFromArray(
     Direct Array Index=$Particles.PartitionParticles.Partition
     Vector Selection Array=DataInterface() {propertyValues {FloatData [<0 0 -0.05>, <0 0 -0.08>]}}
    )
    Use Rotation Vector=true
   ) {module=InitialMeshOrientation002}
  ]
  particleUpdateScript [
   &ParticleState(Kill Particles When Lifetime Has Elapsed=false, Let Infinitely Lived Particles Die When Emitter Deactivates=true)
   &ScaleMeshSize(Scale Factor=&Multiply_VectorByFloat(Vector=<1 1 1>, Float=0.3))
   &SkeletalMeshLocation(
    Bone Sampling Mode=::NewEnumerator4
    MaskSpawnGroup=false
    Sampled Position Offset=&SelectVectorFromArray(
     Direct Array Index=$Particles.PartitionParticles.Partition
     Vector Selection Array=DataInterface() {propertyValues {FloatData [<1 0 3>, <2 2 -6>]}}
    )
    Sampled Position Offset Coordinate Space:ENiagaraSkelMeshTransforms=::NewEnumerator2
    Skeletal Mesh=$System.Skeletal Mesh Reader
   )
   &ScaleColor(
    Scale Alpha=2
    Scale RGB=&MakeVectorFromLinearColorRGB(Color=&SelectLinearColorFromArray(
      Array Sampling Mode=::NewEnumerator1
      Direct Array Index=&MakeIntFromFloat(Float=$User.VariantSwitch)
      Color Selection Array=DataInterface() {propertyValues {ColorData [<rgba 10 10 10 1>, <rgba 0.7292 3.8148 10 1>, <rgba 10 10 10 1>]}}
      Random Seed=?
     ))
    ScaleA=true
    ScaleRGB=true
   )
   &DynamicMaterialParameters(
    Index 0 Param 1=1
    Index 0 Param 2=&SelectFloatFromArray(Direct Array Index=$Particles.PartitionParticles.Partition, Float Selection Array=DataInterface() {propertyValues {FloatData [0.8,1]}})
    Index 0 Param 3=1
    Index 0 Param 4=1
    Index 3 Param 1=1
    Index 3 Param 2=1
    Index 3 Param 3=1
    Index 3 Param 4=&RandomRangeFloat(Minimum=0.5, Maximum=1.2)
    Module.Index0Float0Write=true
    Module.Index0Float1Write=true
    Module.Index0Float2Write=true
    Module.Index0Float3Write=true
    Module.Index3Float0Write=true
    Module.Index3Float1Write=true
    Module.Index3Float2Write=true
    Module.Index3Float3Write=true
    Write Parameter Index 3=::NewEnumerator1
   )
   /SetVariables(Particles.Color_Dark:LinearColor=&MakeLinearColorFromVectorAndFloat(
     "Vector (RGB)"=&MakeVectorFromLinearColorRGB(Color=&SelectLinearColorFromArray(
       Array Sampling Mode=::NewEnumerator1
       Direct Array Index=&MakeIntFromFloat(Float=$User.VariantSwitch)
       Color Selection Array=DataInterface() {propertyValues {ColorData [<rgba 0.0917 0.9497 2 1>, <rgba 0.1324 0.5453 2 1>, <rgba 2 0.6587 0 1>]}}
       Random Seed=?
      ))
     "Float (Alpha)"=0.35
    )) {isSetParameters=true}
   /SetVariables(Particles.Color_Light:LinearColor=&MakeLinearColorFromVectorAndFloat(
     "Vector (RGB)"=&MakeVectorFromLinearColorRGB(Color=&SelectLinearColorFromArray(
       Array Sampling Mode=::NewEnumerator1
       Direct Array Index=&MakeIntFromFloat(Float=$User.VariantSwitch)
       Color Selection Array=DataInterface() {propertyValues {ColorData [<rgba 3.878 20 7.9085 1>, <rgba 3 15.388 30 1>, <rgba 20 15.9179 2 1>]}}
       Random Seed=?
      ))
     "Float (Alpha)"=&Multiply_Float(A=1, B=&FNTimeOfDayScalar(Night Scale=1.2, Morning Scale=0.8, Evening Scale=0.8))
    )) {isSetParameters=true}
   /SetVariables(Particles.Color_glow:LinearColor=&MakeLinearColorFromVectorAndFloat(
     "Vector (RGB)"=&MakeVectorFromLinearColorRGB(Color=&SelectLinearColorFromArray(
       Array Sampling Mode=::NewEnumerator1
       Direct Array Index=&MakeIntFromFloat(Float=$User.VariantSwitch)
       Color Selection Array=DataInterface() {propertyValues {ColorData [<rgba 0 20 12 1>, <rgba 0.3918 16.4575 25 1>, <rgba 12 9.0551 0.5446 1>]}}
       Random Seed=?
      ))
     "Float (Alpha)"=&Multiply_Float(
      A=&Multiply_Float(A=0.35, B=&FNTimeOfDayScalar(Night Scale=1, Morning Scale=0.8, Evening Scale=0.8))
      B=&FN_FrontEndFloatModifier(Gameplay=1, Front End=0.8)
     )
    )) {isSetParameters=true}
  ]
  rendererClasses [&NiagaraMeshRendererProperties]
  renderers [
   {rendererClass=&NiagaraMeshRendererProperties, rendererIndex=0}
   {rendererClass=&NiagaraMeshRendererProperties, rendererIndex=1}
   {rendererClass=&NiagaraMeshRendererProperties, rendererIndex=2}
   {rendererClass=&NiagaraMeshRendererProperties, rendererIndex=3}
  ]
 }
}
summary {
 systemName=NS_Pickaxe_PhoenixKick_IdleR
 emitters [
  {emitterName=headwater_R, rendererClasses [&NiagaraMeshRendererProperties]}
  {emitterName=Light, rendererClasses [&NiagaraLightRendererProperties, &NiagaraSpriteRendererProperties]}
  {emitterName=Sparks, rendererClasses [&NiagaraSpriteRendererProperties]}
  {emitterName=Water, rendererClasses [&NiagaraSpriteRendererProperties]}
 ]
 userVariables [
  User.bIsFrontEnd {type=NiagaraBool}
  User.bIsFrontEndPreview {type=NiagaraBool}
  User.VariantSwitch {type=NiagaraFloat, defaultValue=2, description=""}
 ]
}
