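# Niagara swing trail effect for the PhoenixKick pickaxe cosmetic.
# Body and Trail draw short-lived anim-trail ribbons, while Ring meshes, GPU sparks and SubUV Water sprites spawn by distance travelled.
# User.VariantSwitch drives color and some alpha settings through three-entry arrays and lerps, and also sets the Body's visibility tag.
# SwingLeft flips the sign of ribbon width, and time-of-day and front-end modifiers adjust brightness, glow or sprite size.
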NiagaraSystem /BRCosmetics/Weapons/FORT_Melee/Pickaxe_PhoenixKick/FX/NS_Pickaxe_PhoenixKick_Trail
dependencies {
 usedDataInterfaces [
  &NiagaraDataInterfaceArrayInt32
  &NiagaraDataInterfaceCurve
  &NiagaraDataInterfaceArrayColor
  &NiagaraDataInterfaceSpriteRendererInfo
  &NiagaraDataInterfaceColorCurve
  &NiagaraDataInterfaceVector2DCurve
  &NiagaraDataInterfaceArrayFloat3
  &NiagaraDataInterfaceArrayFloat
  &NiagaraDataInterfaceSocketReader
  &NiagaraDataInterfaceSkeletalMesh
 ]
 usedDynamicInputs []
 usedModules [
  &self:SystemSpawnScript.SystemScriptSource.SystemScriptGraph.NiagaraNodeAssignment_0.SetVariables_#1
  &self:SystemSpawnScript.SystemScriptSource.SystemScriptGraph.NiagaraNodeAssignment_1.SetVariables_#2
  &SystemState
 ]
}
emitters {
 Body {
  emitterUpdateScript [
   &EmitterState(Life Cycle Mode=::NewEnumerator1, Loop Behavior=::NewEnumerator1, Loop Duration=0.3, UseLoopDelay=false)
   &FN_SpawnAnimTrail()
  ]
  particleSpawnScript [
   &FN_InitializeAnimTrailBoneIndecies()
   &FN_InitializeAnimTrail(AnimTrailWidth=15, Color=<rgba 0.52 0.52 0.52 1>, Lifetime=0.2)
   /SetVariables(Particles.VisibilityTag=&SelectIntFromArray(
     Direct Array Index=&MakeIntFromFloat(Float=$User.VariantSwitch)
     Int Selection Array=DataInterface() {propertyValues {IntData [0,1,2]}}
    )) {isSetParameters=true}
  ]
  particleUpdateScript [
   &FN_UpdateAnimTrail()
   &ScaleColor(
    Scale Alpha=&FloatFromCurve(
     FloatCurve=DataInterface() {propertyValues {Curve {keys [<at 0 1 RCIM_Cubic>, {interpMode=RCIM_Cubic, leaveTangent=-2.222, tangentMode=RCTM_Break, time=0.5, value=1}, <at 1 0>]}}}
     Scale Curve=5
    )
    Scale RGB=&Multiply_VectorByFloat(Vector=<1 1 1>, Float=&FNTimeOfDayScalar(Night Scale=0.3, Morning Scale=0.8, Evening Scale=0.7))
    ScaleA=true
    ScaleRGB=true
   )
   &ScaleRibbonWidth(Ribbon Width Scale=&FloatFromCurve(
     FloatCurve=DataInterface() {propertyValues {Curve {keys [<at 0 1 RCIM_Cubic>, {interpMode=RCIM_Cubic, leaveTangent=-6.5138, tangentMode=RCTM_Break, time=0.6477, value=1}]}}}
     Scale Curve=&MakeCustomFloatFromBool(Bool=$User.SwingLeft, True Float=-1, False Float=1)
    ))
   /SetVariables(Particles.Color_Light:LinearColor=&MakeLinearColorFromVectorAndFloat(
     "Vector (RGB)"=&Multiply_VectorByFloat(
      Vector=&MakeVectorFromLinearColorRGB(Color=<rgba 1 1 1 1>)
      Float=&FNTimeOfDayScalar(Night Scale=0.4, Morning Scale=0.8, Evening Scale=0.7)
     )
     "Float (Alpha)"=&FloatFromCurve(
      FloatCurve=DataInterface() {propertyValues {Curve {keys [<at 0 1 RCIM_Cubic>, {interpMode=RCIM_Cubic, leaveTangent=-2.222, tangentMode=RCTM_Break, time=0.5, value=1}, <at 1 0>]}}}
      Scale Curve=5
     )
    )) {isSetParameters=true}
  ]
  rendererClasses [&NiagaraRibbonRendererProperties]
  renderers [
   {rendererClass=&NiagaraRibbonRendererProperties}
   {rendererClass=&NiagaraRibbonRendererProperties, rendererIndex=1}
   {rendererClass=&NiagaraRibbonRendererProperties, rendererIndex=2}
  ]
 }
 Ring {
  emitterUpdateScript [
   &EmitterState(Life Cycle Mode=::NewEnumerator1, Loop Behavior=::NewEnumerator1, Loop Delay=0.05, Loop Duration=0.3, UseLoopDelay=true)
   &SpawnPerUnit(Module.Use Max Movement Threshold=true, Module.Use Movement Tolerance=false, Spawn Spacing=35, Use Spawn Probability=false)
   /SetVariables(Emitter.Color_MainColorB_V:LinearColor=&SelectLinearColorFromArray(
     Array Sampling Mode=::NewEnumerator1
     Direct Array Index=&MakeIntFromFloat(Float=$User.VariantSwitch)
     Color Selection Array=DataInterface() {propertyValues {ColorData [<rgba 0.0454 1 0.2773 1>, <rgba 1 1 1 1>, <rgba 1 0.7652 0.3316 1>]}}
     Random Seed=?
    )) {isSetParameters=true}
   /SetVariables(Emitter.Color_MainColorA_V:LinearColor=&SelectLinearColorFromArray(
     Array Sampling Mode=::NewEnumerator1
     Direct Array Index=&MakeIntFromFloat(Float=$User.VariantSwitch)
     Color Selection Array=DataInterface() {propertyValues {ColorData [<rgba 0 0.8579 2 1>, <rgba 0 0.8579 2 1>, <rgba 2 1.182 0.0862 1>]}}
     Random Seed=?
    )) {isSetParameters=true}
   /SetVariables(Emitter.Color_LightColorB_V:LinearColor=&SelectLinearColorFromArray(
     Array Sampling Mode=::NewEnumerator1
     Direct Array Index=&MakeIntFromFloat(Float=$User.VariantSwitch)
     Color Selection Array=DataInterface() {propertyValues {ColorData [<rgba 0.0858 1.2 0.4373 1>, <rgba 1 1 1 1>, <rgba 1 1 1 1>]}}
     Random Seed=?
    )) {isSetParameters=true}
   /SetVariables(Emitter.Color_LightColorA_V:LinearColor=&SelectLinearColorFromArray(
     Array Sampling Mode=::NewEnumerator1
     Direct Array Index=&MakeIntFromFloat(Float=$User.VariantSwitch)
     Color Selection Array=DataInterface() {propertyValues {ColorData [<rgba 0.0374 0.2375 1 1>, <rgba 0.0374 0.2375 1 1>, <rgba 1 0.7207 0.1094 1>]}}
     Random Seed=?
    )) {isSetParameters=true}
  ]
  particleSpawnScript [
   &InitializeParticle(
    Color Mode=::NewEnumerator0
    Lifetime Max=0.25
    Lifetime Min=0.2
    Lifetime Mode=::NewEnumerator1
    Mesh Scale=&VectorFromFloat(Value=&MakeFloatFromVector2d(Vector 2D=&RandomRangeVector2D(Minimum=<0.15 0>, Maximum=<0.2 0>)))
    Mesh Scale Mode=::NewEnumerator4
    UsePositionOffset=false
    Write Material Random=true
   )
   &ShapeLocation(MaskSpawnGroup=false, Offset Mode=::NewEnumerator0, Override Seed=false, Sphere Radius=2, UseOwnerScale=true)
   &InitialMeshOrientation(
    Mesh Orientation Mode=::NewEnumerator2
    Rotation=&RandomRangeVector(Minimum=<0 0 0>, Maximum=<0 0 1>)
    Use Rotation Vector=true
   )
   &InitialMeshOrientation(Mesh Orientation Mode=::NewEnumerator2, Rotation=<0.25 0 0>, Use Rotation Vector=true) {module=InitialMeshOrientation001}
   &InitialMeshOrientation(
    Mesh Orientation Mode=::NewEnumerator2
    Rotation=&SelectVectorFromArray(
     Direct Array Index=$Particles.PartitionParticles.Partition
     Vector Selection Array=DataInterface() {propertyValues {FloatData [<0 0.12 0>, <0 -0.13 0>]}}
    )
    Use Rotation Vector=true
   ) {enabled=false, module=InitialMeshOrientation002}
   &AddVelocity(Velocity=<1 0 0>)
  ]
  particleUpdateScript [
   &ParticleState()
   &ScaleMeshSize(Scale Factor=&MakeVector(
     X=&Multiply_Float(
      A=1
      B=&FloatFromCurve(FloatCurve=DataInterface() {propertyValues {Curve {keys [<at 0 1>,<at 1 1.35>]}}})
     )
     Y=&Multiply_Float(
      A=1
      B=&FloatFromCurve(FloatCurve=DataInterface() {propertyValues {Curve {keys [<at 0 1>,<at 1 1.35>]}}})
     )
     Z=0.6
    ))
   &UpdateMeshOrientation(Rotation Rate=&RandomRangeFloat(Minimum=-0.2, Maximum=-0.25), Rotation Vector=<0 0 1>) {enabled=false}
   &ScaleColor(
    Scale Alpha=&FloatFromCurve(
     FloatCurve=DataInterface() {propertyValues {Curve {keys [
         <at 0 0 RCIM_Cubic>
         {arriveTangent=-0.0013, interpMode=RCIM_Cubic, leaveTangent=-0.0013, tangentMode=RCTM_User, time=0.1, value=1}
         <at 1 0 RCIM_Cubic>
        ]}}}
     Scale Curve=3
    )
    Scale RGB=&MakeVectorFromLinearColorRGB(Color=&Multiply_LinearColorByFloat(
      Linear Color=&SelectLinearColorFromArray(
       Array Sampling Mode=::NewEnumerator1
       Direct Array Index=&MakeIntFromFloat(Float=$User.VariantSwitch)
       Color Selection Array=DataInterface() {propertyValues {ColorData [<rgba 10 10 10 1>, <rgba 1.9167 4.3887 8 1>, <rgba 7 5.9647 3.2369 1>]}}
       Random Seed=?
      )
      Float=&FNTimeOfDayScalar(Night Scale=0.6, Morning Scale=1, Evening Scale=0.7)
     ))
    ScaleA=true
    ScaleRGB=true
   )
   /SetVariables(Particles.Color_Light:LinearColor=&MakeLinearColorFromVectorAndFloat(
     "Vector (RGB)"=&MakeVectorFromLinearColorRGB(Color=&SelectLinearColorFromArray(
       Array Sampling Mode=::NewEnumerator1
       Direct Array Index=&MakeIntFromFloat(Float=$User.VariantSwitch)
       Color Selection Array=DataInterface() {propertyValues {ColorData [<rgba 9.6949 50 19.7712 1>, <rgba 6.7708 52.5215 100 1>, <rgba 50 38.2578 0 1>]}}
       Random Seed=?
      ))
     "Float (Alpha)"=&Multiply_Float(
      A=&FloatFromCurve(
       FloatCurve=DataInterface() {propertyValues {Curve {keys [
           <at 0 0 RCIM_Cubic>
           {arriveTangent=-0.9177, interpMode=RCIM_Cubic, leaveTangent=-0.9177, tangentMode=RCTM_User, time=0.1, value=1}
           <at 1 0 RCIM_Cubic>
          ]}}}
       Scale Curve=&SelectFloatFromArray(
        Direct Array Index=&MakeIntFromFloat(Float=$User.VariantSwitch)
        Float Selection Array=DataInterface() {propertyValues {FloatData [1,1,5]}}
       )
      )
      B=&FNTimeOfDayScalar(Night Scale=0.7, Morning Scale=1, Evening Scale=0.9)
     )
    )) {isSetParameters=true}
   &SolveForcesAndVelocity(Clamp Velocity=false, Limit Acceleration=false)
  ]
  rendererClasses [&NiagaraMeshRendererProperties]
  renderers [
   {rendererClass=&NiagaraMeshRendererProperties, rendererIndex=0}
   {rendererClass=&NiagaraMeshRendererProperties, rendererIndex=1}
  ]
 }
 Sparks001 {
  simTarget=GPUComputeSim
  emitterUpdateScript [
   &EmitterState(Life Cycle Mode=::NewEnumerator1, Loop Behavior=::NewEnumerator1, Loop Duration=0.15, UseLoopDelay=false)
   &SpawnPerUnit(Module.Use Max Movement Threshold=true, Module.Use Movement Tolerance=false, Spawn Spacing=8, Use Spawn Probability=false)
   /SetVariables(Emitter.Color_BaseColor:LinearColor=&SelectLinearColorFromArray(
     Array Sampling Mode=::NewEnumerator1
     Direct Array Index=&MakeIntFromFloat(Float=$User.VariantSwitch)
     Color Selection Array=DataInterface() {propertyValues {ColorData [<rgba 0.0175 0.0546 0.06 1>, <rgba 0.0175 0.0546 0.06 1>, <rgba 0.1 0.0525 0.0087 1>]}}
     Random Seed=?
    )) {isSetParameters=true}
   /SetVariables(Emitter.Color_DarkColor:LinearColor=&SelectLinearColorFromArray(
     Array Sampling Mode=::NewEnumerator1
     Direct Array Index=&MakeIntFromFloat(Float=$User.VariantSwitch)
     Color Selection Array=DataInterface() {propertyValues {ColorData [<rgba 0.0024 0.0165 0.02 1>, <rgba 0.0024 0.0165 0.02 1>, <rgba 0.02 0.0128 0.002 1>]}}
     Random Seed=?
    )) {isSetParameters=true}
   /SetVariables(Emitter.Color_LightColor:LinearColor=&SelectLinearColorFromArray(
     Array Sampling Mode=::NewEnumerator1
     Direct Array Index=&MakeIntFromFloat(Float=$User.VariantSwitch)
     Color Selection Array=DataInterface() {propertyValues {ColorData [<rgba 0.7604 2 1.6737 1>, <rgba 0.7604 2 1.6737 1>, <rgba 5 4.4565 1.3377 1>]}}
     Random Seed=?
    )) {isSetParameters=true}
  ]
  particleSpawnScript [
   &InitializeParticle(
    Color=<rgba 5 5 5 1>
    Lifetime Max=0.4
    Lifetime Min=0.2
    Lifetime Mode=::NewEnumerator1
    Sprite Size Max=<3.5 10>
    Sprite Size Min=<1.5 5>
    Sprite Size Mode=::NewEnumerator2
    UsePositionOffset=false
    Write Material Random=true
   )
   &ShapeLocation(MaskSpawnGroup=false, Override Seed=false, Sphere Radius=10, UseOwnerScale=true)
   &AddVelocity(Module.Use Falloff Distance=false, Override Seed=false, Velocity Mode=::NewEnumerator1, Velocity Speed=50)
   &InheritVelocity(
    Inherited Velocity Speed Limit:NiagaraFloat=&RandomRangeFloat(Minimum=150, Maximum=180)
    Speed Limit=true
    UseSpeedThreshold=false
   )
  ]
  particleUpdateScript [
   &ParticleState()
   &ScaleSpriteSize(
    Uniform Curve Scale=&FloatFromCurve(
     FloatCurve=DataInterface() {propertyValues {Curve {keys [
         <at 200 0.25 RCIM_Cubic>
         <at 1000 1>
         {arriveTangentWeight=500, interpMode=RCIM_Cubic, leaveTangentWeight=833.3334, tangentMode=RCTM_User, time=2500, value=1}
         <at 5000 1.4>
        ]}}}
     CurveIndex=$Engine.Owner.LODDistance
     Scale Curve=&RandomRangeFloat(Minimum=8, Maximum=10)
    )
    Uniform Curve Sprite Scale=DataInterface() {propertyValues {Curve {keys [
        {arriveTangent=40.4327, interpMode=RCIM_Cubic, leaveTangent=40.4327, tangentMode=RCTM_User, time=0, value=0.7}
        {arriveTangent=-9.8929, interpMode=RCIM_Cubic, leaveTangent=-9.8929, tangentMode=RCTM_User, time=0.07, value=1.5}
        <at 0.16 0.8>
        {arriveTangent=-0.7678, interpMode=RCIM_Cubic, leaveTangent=-0.7678, tangentMode=RCTM_User, time=0.7, value=0.575}
        {arriveTangent=-2.9276, interpMode=RCIM_Cubic, leaveTangent=-2.9276, tangentMode=RCTM_User, time=1, value=0}
       ]}}}
   )
   &ScaleSpriteSize(
    Non-Uniform Curve Sprite Scale=DataInterface() {propertyValues {
      XCurve {keys [{tangentMode=RCTM_User, time=0, value=1}, {tangentMode=RCTM_User, time=1, value=1}]}
      YCurve {keys [{tangentMode=RCTM_User, time=0, value=1}, <at 0.5 1>, {tangentMode=RCTM_User, time=1, value=0}]}
     }}
    Scale Sprite Size Mode=::NewEnumerator3
   ) {module=ScaleSpriteSize001}
   &ScaleColor(
    Scale Alpha=&FloatFromCurve(FloatCurve=DataInterface() {propertyValues {Curve {keys [<at 0.3 1>,<at 0.4 1>,<at 1 1>]}}})
    Scale RGB=/Niagara/DynamicInputs/Math/LerpMultipleVectors(
     Alpha:NiagaraFloat=&Divide_Float(A=$User.VariantSwitch, B=2)
     V 1:Vector3f=&MakeVectorFromLinearColorRGB(Color=&ColorFromCurve(ColorCurve=DataInterface() {propertyValues {
         AlphaCurve {keys []}
         BlueCurve {keys [<at 0.454 0.2866 RCIM_Cubic>,<at 0.6236 1 RCIM_Cubic>]}
         GreenCurve {keys [<at 0.454 1 RCIM_Cubic>,<at 0.6236 1 RCIM_Cubic>]}
         RedCurve {keys [<at 0.454 0.2188 RCIM_Cubic>,<at 0.6236 1 RCIM_Cubic>]}
        }}))
     V 2:Vector3f=&MakeVectorFromLinearColorRGB(Color=&ColorFromCurve(ColorCurve=DataInterface() {propertyValues {
         AlphaCurve {keys []}
         BlueCurve {keys [<at 0.454 2 RCIM_Cubic>,<at 0.6236 1 RCIM_Cubic>]}
         GreenCurve {keys [<at 0.454 0.8986 RCIM_Cubic>,<at 0.6236 1 RCIM_Cubic>]}
         RedCurve {keys [<at 0.454 0.1516 RCIM_Cubic>,<at 0.6236 1 RCIM_Cubic>]}
        }}))
     V 3:Vector3f=&MakeVectorFromLinearColorRGB(Color=&ColorFromCurve(ColorCurve=DataInterface() {propertyValues {AlphaCurve {keys []}, BlueCurve {keys [<at 0.132 1.5>]}, GreenCurve {keys [<at 0.132 1.5>]}, RedCurve {keys [<at 0.132 1.5>]}}}))
     V 4:Vector3f=&MakeVectorFromLinearColorRGB(Color=&ColorFromCurve(ColorCurve=DataInterface() {propertyValues {
         AlphaCurve {keys []}
         BlueCurve {keys [<at 0 0 RCIM_Cubic>,<at 0.132 0>,<at 0.52 0 RCIM_Cubic>]}
         GreenCurve {keys [<at 0 0.2552 RCIM_Cubic>,<at 0.132 1>,<at 0.52 0.1406 RCIM_Cubic>]}
         RedCurve {keys [<at 0 0.145 RCIM_Cubic>,<at 0.132 0.8196>,<at 0.52 0.062 RCIM_Cubic>]}
        }}))
     V 5:Vector3f=&MakeVectorFromLinearColorRGB(Color=&ColorFromCurve(ColorCurve=DataInterface() {propertyValues {
         AlphaCurve {keys []}
         BlueCurve {keys [<at 0 0.2552 RCIM_Cubic>,<at 0.132 1>,<at 0.52 0.1042 RCIM_Cubic>]}
         GreenCurve {keys [<at 0 0.0325 RCIM_Cubic>,<at 0.132 0.1094>,<at 0.52 0.0553 RCIM_Cubic>]}
         RedCurve {keys [<at 0 0.0577 RCIM_Cubic>,<at 0.132 0.5547>,<at 0.52 0.0607 RCIM_Cubic>]}
        }}))
     Vector Count:ENiagara_MultipleLerpCount=::NewEnumerator0
    )
    ScaleA=true
    ScaleRGB=true
   )
   &Drag(Drag=2, Use Linear Drag=true, Use Rotational Drag=false)
   &SolveForcesAndVelocity(Clamp Velocity=false, Limit Acceleration=false)
   &ScaleColor(
    Scale Alpha=&FloatFromCurve(FloatCurve=DataInterface() {propertyValues {Curve {keys [<at 0.3 1>,<at 0.4 1>,<at 1 1>]}}})
    Scale RGB=&MakeVectorFromLinearColorRGB(Color=&LerpLinearColors(
      StartColor=&ColorFromCurve(ColorCurve=DataInterface() {propertyValues {
         AlphaCurve {keys []}
         BlueCurve {keys [<at 0.454 0.2866 RCIM_Cubic>,<at 0.6236 1 RCIM_Cubic>]}
         GreenCurve {keys [<at 0.454 1 RCIM_Cubic>,<at 0.6236 1 RCIM_Cubic>]}
         RedCurve {keys [<at 0.454 0.2188 RCIM_Cubic>,<at 0.6236 1 RCIM_Cubic>]}
        }})
      EndColor=&ColorFromCurve(ColorCurve=DataInterface() {propertyValues {
         AlphaCurve {keys []}
         BlueCurve {keys [<at 0.454 2 RCIM_Cubic>,<at 0.6236 1 RCIM_Cubic>]}
         GreenCurve {keys [<at 0.454 0.8986 RCIM_Cubic>,<at 0.6236 1 RCIM_Cubic>]}
         RedCurve {keys [<at 0.454 0.1516 RCIM_Cubic>,<at 0.6236 1 RCIM_Cubic>]}
        }})
      LerpFactor=$User.VariantSwitch
     ))
    ScaleA=true
    ScaleRGB=true
   ) {enabled=false, module=ScaleColor001}
  ]
  rendererClasses [&NiagaraSpriteRendererProperties]
  renderers [{rendererClass=&NiagaraSpriteRendererProperties}]
 }
 Trail {
  emitterUpdateScript [
   &EmitterState(Inactive Response=::NewEnumerator1, Life Cycle Mode=::NewEnumerator1, Loop Behavior=::NewEnumerator1, Loop Duration=0.3, UseLoopDelay=false)
   &FN_SpawnAnimTrail()
   /SetVariables(Emitter.Color_MainColorA_V:LinearColor=&SelectLinearColorFromArray(
     Array Sampling Mode=::NewEnumerator1
     Direct Array Index=&MakeIntFromFloat(Float=$User.VariantSwitch)
     Color Selection Array=DataInterface() {propertyValues {ColorData [<rgba 0 0.8579 2 1>, <rgba 0 0.8579 2 1>, <rgba 2 1.3761 0.1537 1>]}}
     Random Seed=?
    )) {isSetParameters=true}
   /SetVariables(Emitter.Color_MainColorB_V:LinearColor=&SelectLinearColorFromArray(
     Array Sampling Mode=::NewEnumerator1
     Direct Array Index=&MakeIntFromFloat(Float=$User.VariantSwitch)
     Color Selection Array=DataInterface() {propertyValues {ColorData [<rgba 0.0454 1 0.2773 1>, <rgba 1 1 1 1>, <rgba 2 1.8477 0.9036 1>]}}
     Random Seed=?
    )) {isSetParameters=true}
   /SetVariables(Emitter.Color_LightColorA_V:LinearColor=&LerpLinearColors(StartColor=<rgba 0.1013 1 0.2922 1>, EndColor=<rgba 1.5 1.5 1.5 1>, LerpFactor=$User.VariantSwitch)) {isSetParameters=true}
   /SetVariables(Emitter.Color_LightColorB_V:LinearColor=&LerpLinearColors(StartColor=<rgba 0.0858 1.2 0.4373 1>, EndColor=<rgba 1.5 1.5 1.5 1>, LerpFactor=$User.VariantSwitch)) {isSetParameters=true}
   /SetVariables(Emitter.Color_GlowColorA_V:LinearColor=&SelectLinearColorFromArray(
     Array Sampling Mode=::NewEnumerator1
     Direct Array Index=&MakeIntFromFloat(Float=$User.VariantSwitch)
     Color Selection Array=DataInterface() {propertyValues {ColorData [<rgba 0 0.8579 2 1>, <rgba 0 0.8579 2 1>, <rgba 2 1.3063 0 1>]}}
     Random Seed=?
    )) {isSetParameters=true}
   /SetVariables(Emitter.Color_GlowColorB_V:LinearColor=&SelectLinearColorFromArray(
     Array Sampling Mode=::NewEnumerator1
     Direct Array Index=&MakeIntFromFloat(Float=$User.VariantSwitch)
     Color Selection Array=DataInterface() {propertyValues {ColorData [<rgba 0.0454 1 0.2773 1>, <rgba 0.0454 1 0.2773 1>, <rgba 1 0.7466 0.0945 1>]}}
     Random Seed=?
    )) {isSetParameters=true}
  ]
  particleSpawnScript [
   &FN_InitializeAnimTrailBoneIndecies()
   &FN_InitializeAnimTrail(AnimTrailWidth=25, Color=<rgba 0.7 0.7 0.7 1>, Lifetime=0.2)
  ]
  particleUpdateScript [
   &FN_UpdateAnimTrail()
   &ScaleRibbonWidth(Ribbon Width Scale=&FloatFromCurve(
     FloatCurve=DataInterface() {propertyValues {Curve {keys [
         {arriveTangent=-0.1094, interpMode=RCIM_Cubic, leaveTangent=-0.1094, tangentMode=RCTM_User, time=0, value=1}
         {interpMode=RCIM_Cubic, leaveTangent=-6.5138, tangentMode=RCTM_Break, time=1, value=0.6}
        ]}}}
     Scale Curve=&MakeCustomFloatFromBool(Bool=$User.SwingLeft, True Float=-1, False Float=1)
    ))
   &ScaleColor(
    Scale Alpha=&FloatFromCurve(
     FloatCurve=DataInterface() {propertyValues {Curve {keys [<at 0 1>,<at 0.3 1>,<at 0.6 0>]}}}
     CurveIndex=$Emitter.Age
     Scale Curve=&SelectFloatFromArray(
      Direct Array Index=&MakeIntFromFloat(Float=$User.VariantSwitch)
      Float Selection Array=DataInterface() {propertyValues {FloatData [3,1.5,2]}}
     )
    )
    Scale RGB=&MakeVectorFromLinearColorRGB(Color=&Multiply_LinearColorByFloat(
      Linear Color=&SelectLinearColorFromArray(
       Array Sampling Mode=::NewEnumerator1
       Direct Array Index=&MakeIntFromFloat(Float=$User.VariantSwitch)
       Color Selection Array=DataInterface() {propertyValues {ColorData [<rgba 10 10 10 1>, <rgba 1.9006 5.5581 10 1>, <rgba 8 6.5361 2.6667 1>]}}
       Random Seed=?
      )
      Float=&FNTimeOfDayScalar(Night Scale=0.6, Morning Scale=1, Evening Scale=0.8)
     ))
    ScaleA=true
    ScaleRGB=true
   )
   /SetVariables(Particles.Color_Light:LinearColor=&MakeLinearColorFromVectorAndFloat(
     "Vector (RGB)"=&LerpMultipleVectors(
      Alpha:NiagaraFloat=&Divide_Float(A=$User.VariantSwitch, B=2)
      V 1:Vector3f=&MakeVectorFromLinearColorRGB(Color=&ColorFromCurve(ColorCurve=DataInterface() {propertyValues {
          AlphaCurve {keys []}
          BlueCurve {keys [<at 0.2622 59.3136 RCIM_Cubic>,<at 0.7004 7.9085>]}
          GreenCurve {keys [<at 0.2622 150 RCIM_Cubic>,<at 0.7004 20>]}
          RedCurve {keys [<at 0.2622 29.0847 RCIM_Cubic>,<at 0.7004 3.878>]}
         }}))
      V 2:Vector3f=&MakeVectorFromLinearColorRGB(Color=&ColorFromCurve(ColorCurve=DataInterface() {propertyValues {AlphaCurve {keys []}, BlueCurve {keys [<at 0.3508 150>]}, GreenCurve {keys [<at 0.3508 42.5722>]}, RedCurve {keys [<at 0.3508 0>]}}}))
      V 3:Vector3f=&MakeVectorFromLinearColorRGB(Color=&ColorFromCurve(ColorCurve=DataInterface() {propertyValues {
          AlphaCurve {keys []}
          BlueCurve {keys [<at 0.2672 3.125>,<at 0.6897 0>]}
          GreenCurve {keys [<at 0.2672 64.8567>,<at 0.6897 1.5029>]}
          RedCurve {keys [<at 0.2672 100>,<at 0.6897 5>]}
         }}))
      V 4:Vector3f=&MakeVectorFromLinearColorRGB(Color=&ColorFromCurve(ColorCurve=DataInterface() {propertyValues {
          AlphaCurve {keys []}
          BlueCurve {keys [<at 0 0 RCIM_Cubic>,<at 0.132 0>,<at 0.52 0 RCIM_Cubic>]}
          GreenCurve {keys [<at 0 0.2552 RCIM_Cubic>,<at 0.132 1>,<at 0.52 0.1406 RCIM_Cubic>]}
          RedCurve {keys [<at 0 0.145 RCIM_Cubic>,<at 0.132 0.8196>,<at 0.52 0.062 RCIM_Cubic>]}
         }}))
      V 5:Vector3f=&MakeVectorFromLinearColorRGB(Color=&ColorFromCurve(ColorCurve=DataInterface() {propertyValues {
          AlphaCurve {keys []}
          BlueCurve {keys [<at 0 0.2552 RCIM_Cubic>,<at 0.132 1>,<at 0.52 0.1042 RCIM_Cubic>]}
          GreenCurve {keys [<at 0 0.0325 RCIM_Cubic>,<at 0.132 0.1094>,<at 0.52 0.0553 RCIM_Cubic>]}
          RedCurve {keys [<at 0 0.0577 RCIM_Cubic>,<at 0.132 0.5547>,<at 0.52 0.0607 RCIM_Cubic>]}
         }}))
      Vector Count:ENiagara_MultipleLerpCount=::NewEnumerator0
     )
     "Float (Alpha)"=&Multiply_Float(A=1, B=&FNTimeOfDayScalar(Night Scale=0.8, Morning Scale=1, Evening Scale=1))
    )) {isSetParameters=true}
   /SetVariables(Particles.Color_glow:LinearColor=&MakeLinearColorFromVectorAndFloat(
     "Vector (RGB)"=&MakeVectorFromLinearColorRGB(Color=&SelectLinearColorFromArray(
       Array Sampling Mode=::NewEnumerator1
       Direct Array Index=&MakeIntFromFloat(Float=$User.VariantSwitch)
       Color Selection Array=DataInterface() {propertyValues {ColorData [<rgba 0 20 12 1>, <rgba 0 9.7865 30 1>, <rgba 20 14.3218 0.2585 1>]}}
       Random Seed=?
      ))
     "Float (Alpha)"=&Multiply_Float(
      A=&Multiply_Float(A=0.6, B=&FNTimeOfDayScalar(Night Scale=0.5, Morning Scale=0.8, Evening Scale=0.8))
      B=&FN_FrontEndFloatModifier(Gameplay=1, Front End=0.65)
     )
    )) {isSetParameters=true}
   /SetVariables(Particles.RibbonWidth_Glow:NiagaraFloat=&Multiply_Float(A=$Particles.RibbonWidth, B=1)) {isSetParameters=true}
  ]
  rendererClasses [&NiagaraRibbonRendererProperties]
  renderers [
   {rendererClass=&NiagaraRibbonRendererProperties}
   {rendererClass=&NiagaraRibbonRendererProperties, rendererIndex=1}
   {rendererClass=&NiagaraRibbonRendererProperties, rendererIndex=2}
  ]
 }
 Water {
  emitterUpdateScript [
   &EmitterState(Life Cycle Mode=::NewEnumerator1, Loop Behavior=::NewEnumerator1, Loop Delay=0.02, Loop Duration=0.15, UseLoopDelay=true)
   &SpawnPerUnit(Module.Use Max Movement Threshold=true, Module.Use Movement Tolerance=false, Spawn Spacing=22, Use Spawn Probability=false)
   /SetVariables(Emitter.Color_MainColorB_V:LinearColor=&SelectLinearColorFromArray(
     Array Sampling Mode=::NewEnumerator1
     Direct Array Index=&MakeIntFromFloat(Float=$User.VariantSwitch)
     Color Selection Array=DataInterface() {propertyValues {ColorData [<rgba 0.0288 0.1 0.0875 1>, <rgba 0.0488 0.1348 0.2 1>, <rgba 0.1 0.0441 0.015 1>]}}
     Random Seed=?
    )) {isSetParameters=true}
   /SetVariables(Emitter.Color_DarkColor:LinearColor=&SelectLinearColorFromArray(
     Array Sampling Mode=::NewEnumerator1
     Direct Array Index=&MakeIntFromFloat(Float=$User.VariantSwitch)
     Color Selection Array=DataInterface() {propertyValues {ColorData [<rgba 0.003 0.02 0.0194 1>, <rgba 0.003 0.02 0.0194 1>, <rgba 0.03 0.0141 0.0029 1>]}}
     Random Seed=?
    )) {isSetParameters=true}
   /SetVariables(Emitter.Color_LightColor:LinearColor=&SelectLinearColorFromArray(
     Array Sampling Mode=::NewEnumerator1
     Direct Array Index=&MakeIntFromFloat(Float=$User.VariantSwitch)
     Color Selection Array=DataInterface() {propertyValues {ColorData [<rgba 0.1589 0.5 0.4102 1>, <rgba 0.9531 3 2.4612 1>, <rgba 1.5 1.0863 0.1719 1>]}}
     Random Seed=?
    )) {isSetParameters=true}
  ]
  particleSpawnScript [
   &InitializeParticle(
    Color=<rgba 0.18 0.18 0.18 1>
    Lifetime Max=0.4
    Lifetime Min=0.3
    Lifetime Mode=::NewEnumerator1
    Sprite Size Mode=::NewEnumerator1
    Sprite UV Mode=::NewEnumerator3
    Uniform Sprite Size Max=150
    Uniform Sprite Size Min=100
    UsePositionOffset=false
    Write Material Random=true
   )
   &SubUVAnimation(
    End Frame Range Override=3
    "Sprite Renderer "=DataInterface() {propertyValues {SpriteRenderer=&self:Water_0.NiagaraSpriteRendererProperties_4}}
    SubUV Animation Mode:ENiagara_SubUVLookupModeV2=::NewEnumerator2
    UseEndFrame=true
    UseStartFrame=false
   )
   &CameraOffset(Camera Offset Amount=-5)
   &AddVelocity() {enabled=false}
   &InheritVelocity(Speed Limit=true, UseSpeedThreshold=false)
  ]
  particleUpdateScript [
   &ParticleState()
   &ScaleSpriteSize(
    Uniform Curve Scale=&FN_FrontEndFloatModifier(Gameplay=0.3, Front End=0.2)
    Uniform Curve Sprite Scale=DataInterface() {propertyValues {Curve {keys [
        {interpMode=RCIM_Cubic, leaveTangent=-0.1324, tangentMode=RCTM_Break, time=0, value=1}
        {arriveTangent=-0.2666, tangentMode=RCTM_User, time=1, value=0.75}
       ]}}}
   )
   &DynamicMaterialParameters(
    Index 0 Param 1=&FloatFromCurve(
     FloatCurve=DataInterface() {propertyValues {Curve {
        defaultValue=0
        keys [<at 0.2 0 RCIM_Cubic>, {arriveTangent=1.0645, interpMode=RCIM_Cubic, leaveTangent=1.0645, tangentMode=RCTM_User, time=1, value=1.5}]
       }}}
     Scale Curve=&RandomRangeFloat(Minimum=1, Maximum=0.5)
    )
    Index 0 Param 2=&FloatFromCurve(FloatCurve=DataInterface() {propertyValues {Curve {keys [<at 0 3>,<at 0.3 1>]}}})
    Index 0 Param 4=&FloatFromCurve(FloatCurve=DataInterface() {propertyValues {Curve {keys [<at 0 0>,<at 1 0>]}}})
    Module.Param0WriteEnabled=true
    Module.Param1WriteEnabled=true
    Module.Param2WriteEnabled=true
    Module.Param3WriteEnabled=true
   )
   &ScaleColor(
    Scale Alpha=&FloatFromCurve(FloatCurve=DataInterface() {propertyValues {Curve {keys [<at 0 0>,<at 0.5051 0.9176>,<at 1 1>]}}})
    Scale RGB=&Multiply_VectorByFloat(
     Vector=&LerpMultipleVectors(
      Alpha:NiagaraFloat=&Divide_Float(A=$User.VariantSwitch, B=2)
      V 1:Vector3f=&MakeVectorFromLinearColorRGB(Color=&ColorFromCurve(ColorCurve=DataInterface() {propertyValues {
          AlphaCurve {keys []}
          BlueCurve {keys [<at 0.4363 0.1906 RCIM_Cubic>,<at 0.5929 0.7344 RCIM_Cubic>]}
          GreenCurve {keys [<at 0.4363 1 RCIM_Cubic>,<at 0.5929 0.4945 RCIM_Cubic>]}
          RedCurve {keys [<at 0.4363 0.1291 RCIM_Cubic>,<at 0.5929 0.1925 RCIM_Cubic>]}
         }}))
      V 2:Vector3f=&MakeVectorFromLinearColorRGB(Color=&ColorFromCurve(ColorCurve=DataInterface() {propertyValues {
          AlphaCurve {keys []}
          BlueCurve {keys [<at 0.2788 1 RCIM_Cubic>,<at 0.4604 0.9115>]}
          GreenCurve {keys [<at 0.2788 0.3134 RCIM_Cubic>,<at 0.4604 0.2275>]}
          RedCurve {keys [<at 0.2788 0.1295 RCIM_Cubic>,<at 0.4604 0.0899>]}
         }}))
      V 3:Vector3f=&MakeVectorFromLinearColorRGB(Color=&ColorFromCurve(ColorCurve=DataInterface() {propertyValues {AlphaCurve {keys []}, BlueCurve {keys [<at 0.132 1>]}, GreenCurve {keys [<at 0.132 1>]}, RedCurve {keys [<at 0.132 1>]}}}))
      V 4:Vector3f=&MakeVectorFromLinearColorRGB(Color=&ColorFromCurve(ColorCurve=DataInterface() {propertyValues {
          AlphaCurve {keys []}
          BlueCurve {keys [<at 0 0 RCIM_Cubic>,<at 0.132 0>,<at 0.52 0 RCIM_Cubic>]}
          GreenCurve {keys [<at 0 0.2552 RCIM_Cubic>,<at 0.132 1>,<at 0.52 0.1406 RCIM_Cubic>]}
          RedCurve {keys [<at 0 0.145 RCIM_Cubic>,<at 0.132 0.8196>,<at 0.52 0.062 RCIM_Cubic>]}
         }}))
      V 5:Vector3f=&MakeVectorFromLinearColorRGB(Color=&ColorFromCurve(ColorCurve=DataInterface() {propertyValues {
          AlphaCurve {keys []}
          BlueCurve {keys [<at 0 0.2552 RCIM_Cubic>,<at 0.132 1>,<at 0.52 0.1042 RCIM_Cubic>]}
          GreenCurve {keys [<at 0 0.0325 RCIM_Cubic>,<at 0.132 0.1094>,<at 0.52 0.0553 RCIM_Cubic>]}
          RedCurve {keys [<at 0 0.0577 RCIM_Cubic>,<at 0.132 0.5547>,<at 0.52 0.0607 RCIM_Cubic>]}
         }}))
      Vector Count:ENiagara_MultipleLerpCount=::NewEnumerator0
     )
     Float=&FloatFromCurve(FloatCurve=DataInterface() {propertyValues {Curve {keys [
          <at 0 10>
          {arriveTangent=-8, interpMode=RCIM_Cubic, leaveTangent=-8, time=0.3718, value=9.2943}
          {arriveTangent=-3.2451, interpMode=RCIM_Cubic, leaveTangent=-3.2451, tangentMode=RCTM_User, time=1, value=2}
         ]}}})
    )
    ScaleA=true
    ScaleRGB=true
   )
   &Drag(Use Linear Drag=true, Use Rotational Drag=true)
   &SolveForcesAndVelocity(Clamp Velocity=false, Limit Acceleration=false)
  ]
  rendererClasses [&NiagaraSpriteRendererProperties]
  renderers [{rendererClass=&NiagaraSpriteRendererProperties}]
 }
}
summary {
 systemName=NS_Pickaxe_PhoenixKick_Trail
 emitters [
  {emitterName=Body, rendererClasses [&NiagaraRibbonRendererProperties]}
  {emitterName=Water, rendererClasses [&NiagaraSpriteRendererProperties]}
  {emitterName=Sparks001, simTarget=GPUComputeSim, rendererClasses [&NiagaraSpriteRendererProperties]}
  {emitterName=Ring, rendererClasses [&NiagaraMeshRendererProperties]}
  {emitterName=Trail, rendererClasses [&NiagaraRibbonRendererProperties]}
 ]
 userVariables [
  User.bIsFrontEnd {type=NiagaraBool}
  User.bIsFrontEndPreview {type=NiagaraBool}
  User.RibbonIDOffset {type=NiagaraInt32}
  User.SwingLeft {type=NiagaraBool, defaultValue=true}
  User.VariantSwitch {type=NiagaraFloat, defaultValue=1, description=""}
 ]
}
